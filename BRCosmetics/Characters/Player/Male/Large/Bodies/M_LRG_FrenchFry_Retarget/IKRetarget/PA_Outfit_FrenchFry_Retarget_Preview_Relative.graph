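# Physics asset for the large male FrenchFry outfit's body, stored with its IK Retarget assets and named as a relative retarget preview.
# One capsule per major bone approximates the bulky body, with the pelvis capsule at zero length acting as a sphere.
# Constraints chain the limbs, spine, neck and head back to the pelvis.
# Thigh, lower-spine and neck-base joints explicitly set swing and twist free.
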PhysicsAsset /BRCosmetics/Characters/Player/Male/Large/Bodies/M_LRG_FrenchFry_Retarget/IKRetarget/PA_Outfit_FrenchFry_Retarget_Preview_Relative
bodies {
 calf_l [{shapeType=Capsule, length=15, radius=13, shapeName=calf_l_capsule, center=<-17.4626 -1.3914 0>, rotation=<rot -90 -0.0003 0>}]
 calf_r [{shapeType=Capsule, length=15, radius=13, shapeName=calf_r_capsule, center=<17.4626 1.3914 0>, rotation=<rot -90 179.9997 0>}]
 foot_l [{shapeType=Capsule, length=23.2553, radius=9.2229, shapeName=foot_l_capsule, center=<-4.68 7.5075 0>, rotation=<rot 0 0 90>}]
 foot_r [{shapeType=Capsule, length=23.2553, radius=9.2229, shapeName=foot_r_capsule, center=<4.68 -7.5075 0>, rotation=<rot 0 0 -90>}]
 hand_l [{shapeType=Capsule, length=9, radius=11, shapeName=hand_l_capsule, center=<7.417 -3.3425 0>, rotation=<rot 90 -9.9998 0>}]
 hand_r [{shapeType=Capsule, length=9, radius=11, shapeName=hand_r_capsule, center=<-7.417 3.3425 0>, rotation=<rot 90 170.0002 0>}]
 head [{shapeType=Capsule, length=14.5, radius=18, shapeName=head_capsule, center=<4.7979 1.3743 0>, rotation=<rot 0 49.9984 -90>}]
 lowerarm_l [{
   shapeType=Capsule
   length=11.2429
   radius=9.2571
   shapeName=lowerarm_l_capsule
   center=<9.0972 0.3092 -0.5425>
   rotation=<rot 90 -179.9992 0>
  }]
 lowerarm_r [{shapeType=Capsule, length=11.2429, radius=9.2571, shapeName=lowerarm_r_capsule, center=<-9.0972 -0.3092 0.5425>, rotation=<rot 90 0.0008 0>}]
 neck_02 [{shapeType=Capsule, length=4, radius=5, shapeName=neck_02_capsule, center=<-1.7834 0.3806 0>, rotation=<rot -90 -0.0001 -0.0002>}]
 pelvis [{shapeType=Capsule, length=0, radius=34.435, shapeName=pelvis_capsule, center=<3.8636 -1.3168 0>, rotation=<rot -90 -0.0007 0>}]
 spine_03 [{shapeType=Capsule, length=19.4143, radius=24.5857, shapeName=Spine_03_capsule, center=<-1.4071 4.5311 0>}]
 spine_05 [{shapeType=Capsule, length=12.2958, radius=20.2042, shapeName=spine_05_capsule, center=<-2.4431 0.3671 0>, rotation=<rot 0 -0.0012 0>}]
 thigh_l [{shapeType=Capsule, length=22.5, radius=13.5, shapeName=thigh_l_capsule, center=<-18.5125 0.0008 0>, rotation=<rot -90 -90 89.9968>}]
 thigh_r [{shapeType=Capsule, length=22.5, radius=13.5, shapeName=thigh_r_capsule, center=<18.5125 -0.0008 0>, rotation=<rot -90 179.9968 0>}]
 upperarm_l [{
   shapeType=Capsule
   length=14.4772
   radius=11.0228
   shapeName=upperarm_l_capsule
   center=<11.1091 0.0001 -1.6327>
   rotation=<rot 0.0001 90.0011 -90>
  }]
 upperarm_r [{
   shapeType=Capsule
   length=14.4772
   radius=11.0228
   shapeName=upperarm_r_capsule
   center=<-11.1091 -0.0001 1.6327>
   rotation=<rot 0.0001 90.0011 90>
  }]
}
constraints [
 {bone1Name=spine_05, bone2Name=spine_03, swing1LimitDegrees=45, swing2LimitDegrees=45, twistLimitDegrees=45}
 {bone1Name=head, bone2Name=neck_02, swing1LimitDegrees=45, swing2LimitDegrees=45, twistLimitDegrees=45}
 {bone1Name=upperarm_l, bone2Name=spine_05, swing1LimitDegrees=45, swing2LimitDegrees=45, twistLimitDegrees=45}
 {bone1Name=lowerarm_l, bone2Name=upperarm_l, swing1LimitDegrees=45, swing2LimitDegrees=45, twistLimitDegrees=45}
 {bone1Name=hand_l, bone2Name=lowerarm_l, swing1LimitDegrees=45, swing2LimitDegrees=45, twistLimitDegrees=45}
 {bone1Name=upperarm_r, bone2Name=spine_05, swing1LimitDegrees=45, swing2LimitDegrees=45, twistLimitDegrees=45}
 {bone1Name=lowerarm_r, bone2Name=upperarm_r, swing1LimitDegrees=45, swing2LimitDegrees=45, twistLimitDegrees=45}
 {bone1Name=hand_r, bone2Name=lowerarm_r, swing1LimitDegrees=45, swing2LimitDegrees=45, twistLimitDegrees=45}
 {bone1Name=calf_l, bone2Name=thigh_l, swing1LimitDegrees=45, swing2LimitDegrees=45, twistLimitDegrees=45}
 {bone1Name=foot_l, bone2Name=calf_l, swing1LimitDegrees=45, swing2LimitDegrees=45, twistLimitDegrees=45}
 {bone1Name=calf_r, bone2Name=thigh_r, swing1LimitDegrees=45, swing2LimitDegrees=45, twistLimitDegrees=45}
 {bone1Name=foot_r, bone2Name=calf_r, swing1LimitDegrees=45, swing2LimitDegrees=45, twistLimitDegrees=45}
 {
  bone1Name=thigh_l
  bone2Name=pelvis
  swing1LimitDegrees=45
  swing1Motion=Free
  swing2LimitDegrees=45
  swing2Motion=Free
  twistLimitDegrees=45
  twistMotion=Free
 }
 {
  bone1Name=thigh_r
  bone2Name=pelvis
  swing1LimitDegrees=45
  swing1Motion=Free
  swing2LimitDegrees=45
  swing2Motion=Free
  twistLimitDegrees=45
  twistMotion=Free
 }
 {
  bone1Name=spine_03
  bone2Name=pelvis
  swing1LimitDegrees=45
  swing1Motion=Free
  swing2LimitDegrees=45
  swing2Motion=Free
  twistLimitDegrees=45
  twistMotion=Free
 }
 {
  bone1Name=neck_02
  bone2Name=spine_05
  swing1LimitDegrees=45
  swing1Motion=Free
  swing2LimitDegrees=45
  swing2Motion=Free
  twistLimitDegrees=45
  twistMotion=Free
 }
]
